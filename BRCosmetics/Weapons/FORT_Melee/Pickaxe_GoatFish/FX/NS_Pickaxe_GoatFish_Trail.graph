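# Niagara swing trail effect for the GoatFish pickaxe, combining ribbon anim trails with underwater-style bubbles.
# MainTrail and SharpTrail draw short-lived ribbons between socket pairs, with swing fade and curve-driven color over life.
# SwingBubbles spawns sprites per distance moved at interpolated socket positions, blends position space over life, drifts them with noise, and drives a late pop effect.
# System time-of-day values adjust trail tint, trail alpha and bubble opacity.
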NiagaraSystem /BRCosmetics/Weapons/FORT_Melee/Pickaxe_GoatFish/FX/NS_Pickaxe_GoatFish_Trail
dependencies {
 usedDataInterfaces [&NiagaraDataInterfaceColorCurve, &NiagaraDataInterfaceCurve, &NiagaraDataInterfaceSocketReader]
 usedDynamicInputs []
 usedModules [
  &self:SystemSpawnScript.SystemScriptSource.SystemScriptGraph.NiagaraNodeAssignment_1.SetVariables_#1
  &SystemState
  &BlendSpaceHelper
  &self:SystemSpawnScript.SystemScriptSource.SystemScriptGraph.NiagaraNodeAssignment_2.SetVariables_#2
  &self:SystemSpawnScript.SystemScriptSource.SystemScriptGraph.NiagaraNodeAssignment_0.SetVariables_#3
  &self:SystemSpawnScript.SystemScriptSource.SystemScriptGraph.NiagaraNodeAssignment_3.SetVariables_#4
  &self:SystemSpawnScript.SystemScriptSource.SystemScriptGraph.NiagaraNodeAssignment_4.SetVariables_#5
 ]
}
emitters {
 MainTrail {
  emitterSpawnScript [&FN_RibbonIDOffsetUtility()]
  emitterUpdateScript [
   &EmitterState()
   &FN_SpawnAnimTrail()
  ]
  particleSpawnScript [
   &FN_InitializeAnimTrailBoneIndecies(First Bone or Socket Index=1, Second Bone or Socket Index=0)
   &FN_InitializeAnimTrailBoneIndecies(AnimTrail Selection=1, First Bone or Socket Index=2, Second Bone or Socket Index=3) {module=FN_InitializeAnimTrailBoneIndecies001}
   &FN_InitializeAnimTrail(
    AnimTrailWidth=1.5
    Color=&MultiplyLinearColor(
     Linear Color=<rgba 1 1 1 0.4>
     Scale Factor=&MakeVector4FromVectorAndFloat("Vector (RGB)"=<1 1 1>, "Float (Alpha)"=$System.SwingGeneralFade)
    )
    Lifetime=0.25
   )
   &AddVelocity(Velocity=<0 400 0>) {enabled=false}
  ]
  particleUpdateScript [
   &SolveForcesAndVelocity(Clamp Velocity=false, Limit Acceleration=false)
   &FN_UpdateAnimTrail() {module=FN_UpdateAnimTrail001}
   &ScaleColor(
    Linear Color Curve=DataInterface() {propertyValues {
      AlphaCurve {keys [<at 0 1>,<at 0.4147 0>]}
      BlueCurve {keys [<at 0 0.9632>,<at 0.4064 0.8149>,<at 0.9472 0.7453>]}
      GreenCurve {keys [<at 0 0.8783>,<at 0.4064 0.6772>,<at 0.9472 0.6316>]}
      RedCurve {keys [<at 0 1>,<at 0.4064 0.693>,<at 0.9472 0.5009>]}
     }}
    Scale Mode=::NewEnumerator2
   )
   &ScaleColor(Scale Alpha=$System.TOD_Alpha, ScaleA=true, ScaleRGB=true) {module=ScaleColor001}
  ]
  rendererClasses [&NiagaraRibbonRendererProperties]
  renderers [
   {rendererClass=&NiagaraRibbonRendererProperties}
   {rendererClass=&NiagaraRibbonRendererProperties, rendererIndex=1}
  ]
 }
 SharpTrail {
  emitterSpawnScript [&FN_RibbonIDOffsetUtility()]
  emitterUpdateScript [
   &EmitterState()
   &FN_SpawnAnimTrail(Number of Anim Trails=3)
  ]
  particleSpawnScript [
   &FN_InitializeAnimTrailBoneIndecies(First Bone or Socket Index=1, Second Bone or Socket Index=0)
   &FN_InitializeAnimTrailBoneIndecies(AnimTrail Selection=1, First Bone or Socket Index=3, Second Bone or Socket Index=2) {module=FN_InitializeAnimTrailBoneIndecies001}
   &FN_InitializeAnimTrailBoneIndecies(AnimTrail Selection=2, First Bone or Socket Index=5, Second Bone or Socket Index=4) {module=FN_InitializeAnimTrailBoneIndecies002}
   &FN_InitializeAnimTrail(
    Color=&MultiplyLinearColor(
     Linear Color=&MergeHelper_ColorSwitch(
      0=<rgba 0.9219 0.9319 1 0.5>
      1=<rgba 1 0.668 0.2708 0.5>
      2=<rgba 1 0.668 0.2708 0.5>
      3=?
      Index=&SpawnGroup()
      Num=::NewEnumerator1
     )
     Scale Factor=&MakeVector4FromVectorAndFloat("Vector (RGB)"=<1 1 1>, "Float (Alpha)"=$System.SwingGeneralFade)
    )
    Lifetime=&MergeHelper_FloatSwitch(0=0.3, 1=0.25, 2=0.25, 3=?, Index=&SpawnGroup(), Num=::NewEnumerator1)
   )
   &AddVelocity(Velocity=<0 400 0>) {enabled=false}
  ]
  particleUpdateScript [
   &SolveForcesAndVelocity(Clamp Velocity=false, Limit Acceleration=false)
   &FN_UpdateAnimTrail() {module=FN_UpdateAnimTrail001}
   &ScaleColor(
    Scale Mode=::NewEnumerator1
    Scale RGBA=&MakeVector4_FromLinearColor(LinearColor=&MergeHelper_ColorSwitch(
      0=&ColorFromCurve(ColorCurve=DataInterface() {propertyValues {
         AlphaCurve {keys [<at 0.0578 1>,<at 0.4517 0>]}
         BlueCurve {keys [<at 0 0.849>,<at 0.4064 0.8149>,<at 0.9472 0.7453>]}
         GreenCurve {keys [<at 0 0.5>,<at 0.4064 0.6772>,<at 0.9472 0.5799>]}
         RedCurve {keys [<at 0 1>,<at 0.4064 0.693>,<at 0.9472 0.3898>]}
        }})
      1=&MultiplyLinearColor(
       Linear Color=&ColorFromCurve(ColorCurve=DataInterface() {propertyValues {
          AlphaCurve {keys [<at 0 1 RCIM_Cubic>,<at 0.3052 0 RCIM_Cubic>]}
          BlueCurve {keys [<at 0 1>,<at 1 0>]}
          GreenCurve {keys [<at 0 1>,<at 1 0>]}
          RedCurve {keys [<at 0 1>,<at 1 0>]}
         }})
       Scale Factor=&MakeVector4_FromLinearColor(LinearColor=$System.TOD_Color)
      )
      2=&MultiplyLinearColor(
       Linear Color=&ColorFromCurve(ColorCurve=DataInterface() {propertyValues {
          AlphaCurve {keys [<at 0 1 RCIM_Cubic>,<at 0.3052 0 RCIM_Cubic>]}
          BlueCurve {keys [<at 0 1>,<at 1 0>]}
          GreenCurve {keys [<at 0 1>,<at 1 0>]}
          RedCurve {keys [<at 0 1>,<at 1 0>]}
         }})
       Scale Factor=&MakeVector4_FromLinearColor(LinearColor=$System.TOD_Color)
      )
      3=?
      Index=&SpawnGroup()
      Num=::NewEnumerator1
     ))
    ScaleRGBA=true
   )
   &ScaleColor(Scale Alpha=$System.TOD_Alpha, ScaleA=true, ScaleRGB=true) {module=ScaleColor001}
  ]
  rendererClasses [&NiagaraRibbonRendererProperties]
  renderers [
   {rendererClass=&NiagaraRibbonRendererProperties}
   {rendererClass=&NiagaraRibbonRendererProperties, rendererIndex=1}
  ]
 }
 SwingBubbles {
  emitterUpdateScript [
   &EmitterState(Life Cycle Mode=::NewEnumerator1, Loop Delay=0.05, Loop Duration=0.25, UseLoopDelay=true)
   &SpawnPerUnit(Module.Use Max Movement Threshold=true, Module.Use Movement Tolerance=false, Spawn Spacing=35, Use Spawn Probability=false)
  ]
  particleSpawnScript [
   &InitializeParticle(
    Color Maximum=<rgba 0.8963 0.6048 0.8641 1>
    Color Minimum=<rgba 0.6174 0.8087 1 2>
    Color Mode=::NewEnumerator2
    Lifetime Max=0.35
    Lifetime Min=0.25
    Lifetime Mode=::NewEnumerator1
    Sprite Rotation Mode=::NewEnumerator2
    Sprite Size Mode=::NewEnumerator3
    Uniform Sprite Size=&Multiply_Float(A=&RandomRangeFloat(Minimum=3, Maximum=8), B=1)
    UsePositionOffset=false
    Write Material Random=true
   )
   &SocketLocation(
    Interpolate Socket Sampling=true
    Orientation Sampling=::NewEnumerator1
    SamplePosOffset=false
    Socket Reader=$System.Socket Reader
    Socket Sampling Mode=::NewEnumerator3
    VelocitySpeedLimit=false
   )
   &AddVelocity(
    Additional Quaternion Rotation=true
    Additional Yaw / Pitch / Roll:Vector3f=<90 0 0>
    Cone Angle=90
    Inner Cone Angle=60
    Module.UseConeFalloff=false
    Module.UseVelDistribution=false
    Override Seed=false
    Rotation Mode=::NewEnumerator3
    Rotation Quaternion:Quat4f=$Output.SocketLocation.SampledRotation
    Velocity Mode=::NewEnumerator2
    Velocity Speed=&RandomRangeFloat(Minimum=25, Maximum=50)
   )
   /SetVariables(Particles.TexIndex:NiagaraFloat=&RandomRangeFloat(Minimum=0, Maximum=3)) {isSetParameters=true}
  ]
  particleUpdateScript [
   &ParticleState()
   &ScaleSpriteSize(Uniform Curve Sprite Scale=DataInterface() {propertyValues {Curve {keys [{interpMode=RCIM_Cubic, tangentMode=RCTM_User, time=0.5958, value=1}, <at 0.6058 1.25 RCIM_Constant>, <at 0.6131 0 RCIM_Constant>]}}})
   &BlendSpace(
    "Position Blend (World <-> Local)":NiagaraFloat=&FloatFromCurve(
     FloatCurve=DataInterface() {propertyValues {Curve {
        defaultValue=1
        keys [
         {arriveTangent=-8.534, interpMode=RCIM_Cubic, leaveTangent=-8.534, tangentMode=RCTM_User, time=0, value=1}
         {interpMode=RCIM_Cubic, tangentMode=RCTM_User, time=0.35, value=0}
        ]
       }}}
     Translation Contribution:NiagaraFloat=1
     Rotation Contribution:NiagaraFloat=1
    )
    Position Coordinate Space=1
   )
   &AccelerationForce(Acceleration=&Multiply_Vector(
     A=<2 2 0.5>
     B=&MakeVector(
      X=&Waveform(
       "Waveform [1]"=::NewEnumerator12
       "[1] Input Position"=&RandomRangeFloat(Minimum=0, Maximum=1003)
       "[1] Frequency"=1
       "[1] Amplitude Min/Max"=<-1 1>
       "[1] Scale"=60
      )
      Y=&Waveform(
       "Waveform [1]"=::NewEnumerator12
       "[1] Input Position"=&RandomRangeFloat(Minimum=0, Maximum=1000)
       "[1] Frequency"=1
       "[1] Amplitude Min/Max"=<-1 1>
       "[1] Scale"=60
      )
      Z=&RandomRangeFloat(Minimum=75, Maximum=100)
     )
    ))
   &CurlNoiseForce(Bias Noise=false, Noise Strength=250, Pan Noise=false)
   &SolveForcesAndVelocity(Clamp Velocity=false, Limit Acceleration=false)
   &ScaleColor(
    Scale Alpha=&Multiply_Float(
     A=&FloatFromCurve(
      FloatCurve=DataInterface() {propertyValues {Curve {keys [
          {interpMode=RCIM_Cubic, tangentMode=RCTM_User, time=0, value=0}
          {interpMode=RCIM_Cubic, tangentMode=RCTM_User, time=0.3, value=1}
          <at 0.95 1.0068>
          {tangentMode=RCTM_User, time=1, value=0}
         ]}}}
      Scale Curve=0.5
     )
     B=&FNTimeOfDayScalar(Night Scale=1, Morning Scale=0.2, Evening Scale=0.5)
    )
    ScaleA=true
    ScaleRGB=true
   )
   /SetVariables(Particles.PopColor:LinearColor=&ColorFromCurve(ColorCurve=DataInterface() {propertyValues {
       AlphaCurve {keys [
         {interpMode=RCIM_Cubic, tangentMode=RCTM_Break, time=0.65, value=0}
         {arriveTangent=-7.696, interpMode=RCIM_Cubic, leaveTangent=-7.696, tangentMode=RCTM_User, time=0.7613, value=1.9016}
         {arriveTangent=-0.4593, interpMode=RCIM_Cubic, leaveTangent=-0.4593, tangentMode=RCTM_User, time=1, value=0}
        ]}
       BlueCurve {keys [<at 0 0.8 RCIM_Cubic>]}
       GreenCurve {keys [<at 0 0.8 RCIM_Cubic>]}
       RedCurve {keys [<at 0 0.8 RCIM_Cubic>]}
      }})) {isSetParameters=true}
   /SetVariables(Particles.RandomScaleFloat:Vector2f=&Multiply_Vector2D_ByFloat(
     Float=&Multiply_Float(
      A=&FloatFromCurve(FloatCurve=DataInterface() {propertyValues {Curve {keys [
           <at 0 0 RCIM_Constant>
           {arriveTangent=3.749, interpMode=RCIM_Cubic, leaveTangent=3.749, tangentMode=RCTM_User, time=0.7, value=1}
           <at 1.0085 1.5056 RCIM_Cubic>
          ]}}})
      B=1.5
     )
     Vector2D=$Particles.Initial.SpriteSize
    )) {isSetParameters=true}
   /SetVariables(Particles.PopDynamic:Vector4f=&MakeVector4(
     X=&FloatFromCurve(FloatCurve=DataInterface() {propertyValues {Curve {keys [<at 0.8 1>,<at 1 0.6>]}}})
     Y=0
     Z=0
     W=0
    )) {isSetParameters=true}
   /SetVariables(Particles.PopPosition:NiagaraPosition=$Particles.Position) {isSetParameters=true}
   /SetVariables(Particles.Color=&MultiplyLinearColor(Linear Color=$Particles.Color, Scale Factor=&MakeVector4(X=1, Y=1, Z=1, W=$System.TOD_Bubbles))) {isSetParameters=true}
  ]
  rendererClasses [&NiagaraSpriteRendererProperties]
  renderers [
   {rendererClass=&NiagaraSpriteRendererProperties}
   {rendererClass=&NiagaraSpriteRendererProperties, rendererIndex=1}
  ]
 }
}
summary {
 systemName=NS_Pickaxe_GoatFish_Trail
 emitters [
  {emitterName=SharpTrail, rendererClasses [&NiagaraRibbonRendererProperties]}
  {emitterName=MainTrail, rendererClasses [&NiagaraRibbonRendererProperties]}
  {emitterName=SwingBubbles, rendererClasses [&NiagaraSpriteRendererProperties]}
 ]
 userVariables [
  User.RibbonIDOffset {type=NiagaraInt32}
  User.SwingSectionIndex {type=NiagaraInt32}
 ]
}
