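# Control Rig for the GoatFish pickaxe that adjusts the pose of its dynamic fish-goat bones by player state.
# A three-branch sequence layers additive-style bone offsets: a constant base pose, then one weighted by smoothed |1 − NotCrouchingandPickaxe|.
# Further offsets follow smoothed CrouchingFloat and the raw front-end flag.
# The ismeleeAttackingFloat variable is declared but unused in the graph.
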ControlRigBlueprint /BRCosmetics/Weapons/FORT_Melee/Pickaxe_GoatFish/Meshes/GoatFish_Axe_CtrlRig
graphs [
 RigVMFunctionLibrary {links [], nodes []}
 RigVMFunction_Sequence_1_ContainedGraph {nodes [
   n1 = RigVMFunctionEntryNode()
   RigVMFunction_Sequence = &RigVMFunction_Sequence(ExecuteContext=@n1.ExecuteContext)
   n2 = &RigVMFunction_Sequence(ExecuteContext=@RigVMFunction_Sequence.B)
   RigVMFunctionReturnNode(A=@RigVMFunction_Sequence.A, B=@n2.A, C=@n2.B)
  ]}
 RigVMModel {nodes [
   RigUnit_BeginExecution = &RigUnit_BeginExecution()
   RigVMFunction_Sequence = RigVMAggregateNode(ExecuteContext=@RigUnit_BeginExecution.ExecutePin)
   &RigUnit_ModifyTransforms(
    ItemToModify="((Item=(Type=Bone,Name=\"dyn_l_Pelvic_fin_01\"),Transform=(Rotation=(X=0,Y=0.342,Z=0,W=0.9397),Translation=(X=0,Y=0,Z=0),Scale3D=(X=1,Y=1,Z=1))),(Item=(Type=Bone,Name=\"dyn_spine_01\"),Transform=(Rotation=(X=0,Y=0,Z=0.1305,W=0.9914),Translation=(X=0,Y=0,Z=0),S ...(truncated)"
    Weight=1
    WeightMinimum=0
    ExecutePin=@RigVMFunction_Sequence.A
   )
   Subtract = &RigVMFunction_MathDoubleSub(A=1, B=RigVMVariableNode() {variable=NotCrouchingandPickaxe}.Value)
   Absolute = &RigVMFunction_MathDoubleAbs(Value=@Subtract.Result)
   n1 = &RigVMFunction_AlphaInterp(InterpSpeedDecreasing=5, InterpSpeedIncreasing=5, Scale=1, bClampResult=False, Value=@Absolute.Result)
   > &RigUnit_ModifyTransforms(
    ItemToModify="((Item=(Type=Bone,Name=\"dyn_spine_01\"),Transform=(Rotation=(X=0,Y=0.0872,Z=0,W=0.9962),Translation=(X=0,Y=0,Z=0),Scale3D=(X=1,Y=1,Z=1))),(Item=(Type=Bone,Name=\"dyn_spine_02\"),Transform=(Rotation=(X=0,Y=0.0872,Z=0,W=0.9962),Translation=(X=0,Y=0,Z=0),Scale3D ...(truncated)"
    WeightMinimum=0
    Weight=@n1.Result
   )
   AlphaInterp = &RigVMFunction_AlphaInterp(InterpSpeedDecreasing=7, InterpSpeedIncreasing=7, Scale=1, bClampResult=False, Value=RigVMVariableNode() {variable=CrouchingFloat}.Value)
   &RigUnit_ModifyTransforms(
    ItemToModify="((Item=(Type=Bone,Name=\"dyn_spine_01\"),Transform=(Rotation=(X=0,Y=0,Z=0.1736,W=0.9848),Translation=(X=0,Y=0,Z=0),Scale3D=(X=1,Y=1,Z=1))),(Item=(Type=Bone,Name=\"dyn_spine_02\"),Transform=(Rotation=(X=0,Y=0,Z=0.0872,W=0.9962),Translation=(X=0,Y=3,Z=0),Scale3D ...(truncated)"
    WeightMinimum=0
    Weight=@AlphaInterp.Result
    ExecutePin=@RigVMFunction_Sequence.B
   )
   > &RigUnit_ModifyTransforms(
    ItemToModify=((Item=(Type=Bone,Name="dyn_l_Pelvic_fin_01"),Transform=(Rotation=(X=0.0076,Y=0.043,Z=-0.1735,W=0.9839),Translation=(X=0,Y=0,Z=0),Scale3D=(X=1,Y=1,Z=1))))
    WeightMinimum=0
    Weight=@AlphaInterp.Result
   )
   &RigUnit_ModifyTransforms(
    ItemToModify="((Item=(Type=Bone,Name=\"dyn_spine_01\"),Transform=(Rotation=(X=0,Y=0.1305,Z=0,W=0.9914),Translation=(X=0,Y=0,Z=0),Scale3D=(X=1,Y=1,Z=1))),(Item=(Type=Bone,Name=\"dyn_spine_02\"),Transform=(Rotation=(X=0,Y=0.1305,Z=0,W=0.9914),Translation=(X=0,Y=0,Z=0),Scale3D ...(truncated)"
    WeightMinimum=0
    Weight=RigVMVariableNode() {variable=isInFrontEndFloat}.Value
    ExecutePin=@RigVMFunction_Sequence.C
   )
  ]}
]
hierarchy [
 b_AR_Root {loc=<0 0 0>, rot=<0 0 0 1>}
 Melee_TwoHanded_Handle {parent=b_AR_Root, loc=<0 0 0>, rot=<0 0 0.7071 -0.7071>}
 dyn_spine_root {parent=Melee_TwoHanded_Handle, loc=<-0.245 -0.3368 2.3419>, rot=<-0.5254 -0.5254 -0.4732 0.4732>}
 dyn_spine_01 {parent=dyn_spine_root, loc=<0 0 0>, rot=<0 0 0 1>}
 dyn_spine_02 {parent=dyn_spine_01, loc=<9.75 0 0>, rot=<0 0.0486 0 -0.9988>}
 dyn_spine_03 {parent=dyn_spine_02, loc=<9.75 0 0>, rot=<0 0.0486 0 -0.9988>}
 dyn_spine_04 {parent=dyn_spine_03, loc=<9.75 0 0>, rot=<0 0.0248 0 -0.9997>}
 dyn_head {parent=dyn_spine_04, loc=<7.2369 0 0>, rot=<0 -0.3863 0 -0.9224>}
 dyn_mouth_01 {parent=dyn_head, loc=<8.6431 0 0.1047>, rot=<0 -0.35 0 -0.9367>}
 dyn_mouth_02 {parent=dyn_mouth_01, loc=<7.9037 0 -0.1859>, rot=<0 -0.7098 0 -0.7044>}
 dyn_mouth_03 {parent=dyn_mouth_02, loc=<3.1608 0 0.1859>, rot=<0.0953 0.062 -0.0398 -0.9927>}
 dyn_mouth_04 {parent=dyn_mouth_03, loc=<3.1769 0 0>, rot=<0 0.4937 0 -0.8697>}
 dyn_l_haorn_base {parent=dyn_head, loc=<4.2655 -2.8939 14.9435>, rot=<0.3649 0.0069 0.9283 -0.0705>}
 dyn_l_haorn_01 {parent=dyn_l_haorn_base, loc=<0 0 0>, rot=<0 0 0 1>}
 dyn_l_haorn_02 {parent=dyn_l_haorn_01, loc=<6.609 0 0>, rot=<-0.0333 -0.326 -0.1394 -0.9344>}
 dyn_l_haorn_03 {parent=dyn_l_haorn_02, loc=<8.065 0 0>, rot=<-0.1374 0.2517 0.1415 -0.9475>}
 dyn_r_haorn_base {parent=dyn_head, loc=<4.2655 2.8939 14.9435>, rot=<0.3649 -0.0069 0.9283 0.0705>}
 dyn_r_haorn_01 {parent=dyn_r_haorn_base, loc=<0 0 0>, rot=<0 0 0 1>}
 dyn_r_haorn_02 {parent=dyn_r_haorn_01, loc=<6.609 0 0>, rot=<0.0333 -0.326 0.1394 -0.9344>}
 dyn_r_haorn_03 {parent=dyn_r_haorn_02, loc=<8.065 0 0>, rot=<0.1374 0.2517 -0.1415 -0.9475>}
 dyn_l_ear_base {parent=dyn_head, loc=<0 -5.475 11.775>, rot=<0.1568 -0.7958 -0.4984 0.306>}
 dyn_l_ear_01 {parent=dyn_l_ear_base, loc=<0 0 0>, rot=<0 0 0 1>}
 dyn_l_ear_02 {parent=dyn_l_ear_01, loc=<6.2522 0 0>, rot=<0 0 0 1>}
 dyn_r_ear_base {parent=dyn_head, loc=<0 5.475 11.775>, rot=<-0.1568 -0.7958 0.4984 0.306>}
 dyn_r_ear_01 {parent=dyn_r_ear_base, loc=<0 0 0>, rot=<0 0 0 1>}
 dyn_r_ear_02 {parent=dyn_r_ear_01, loc=<6.2522 0 0>, rot=<0 0 0 1>}
 dyn_fin_base {parent=dyn_spine_04, loc=<1.7923 0 14.0642>, rot=<0.4266 0 0.9045 0>}
 dyn_fin {parent=dyn_fin_base, loc=<0 0 0>, rot=<0 0 0 1>}
 dyn_l_Pelvic_fin_base {parent=dyn_spine_03, loc=<0.8506 -9.3316 -3.9914>, rot=<-0.1517 -0.4173 0.7574 -0.4788>}
 dyn_l_Pelvic_fin_01 {parent=dyn_l_Pelvic_fin_base, loc=<0 0 0>, rot=<0 0 0 1>}
 dyn_l_Pelvic_fin_02 {parent=dyn_l_Pelvic_fin_01, loc=<2.921 0 0>, rot=<0.0006 0.0058 0.1105 -0.9939>}
 dyn_r_Pelvic_fin_base {parent=dyn_spine_03, loc=<0.8506 9.3316 -3.9914>, rot=<-0.1517 0.4173 0.7574 0.4788>}
 dyn_r_Pelvic_fin_01 {parent=dyn_r_Pelvic_fin_base, loc=<0 0 0>, rot=<0 0 0 1>}
 dyn_r_Pelvic_fin_02 {parent=dyn_r_Pelvic_fin_01, loc=<2.921 0 0>, rot=<-0.0006 0.0058 -0.1105 -0.9939>}
 dyn_tail_base {parent=Melee_TwoHanded_Handle, loc=<-0.245 0.1924 -5.4189>, rot=<0.3055 -0.3055 -0.6377 -0.6377>}
 dyn_tail_a__01 {parent=dyn_tail_base, loc=<3.1215 0 0>, rot=<0 -0.0723 0 -0.9974>}
 dyn_tail_b__01 {parent=dyn_tail_base, loc=<0.75 0 -3>, rot=<0.7896 -0.0002 -0.6136 0.0024>}
 body_root {parent=b_AR_Root, loc=<0 0 0>, rot=<-0.0003 0.0039 0.0006 -1>}
 pelvis {parent=body_root, loc=<0 -0.7071 93.1191>, rot=<0.0127 0.707 -0.0127 -0.707>}
 thigh_l {parent=pelvis, loc=<-2.5693 -0.0109 -7.7501>, rot=<0.0012 0.0572 0.0155 -0.9982>}
 calf_l {parent=thigh_l, loc=<-39.0484 0 0>, rot=<0.0006 -0.0161 -0.0391 -0.9991>}
 thigh_r {parent=pelvis, loc=<-2.5691 -0.0109 7.7501>, rot=<0.0572 -0.0012 0.9982 0.0155>}
 calf_r {parent=thigh_r, loc=<39.0487 0 0>, rot=<0.0006 -0.0161 -0.0391 -0.9991>}
 spine_01 {parent=pelvis, loc=<6.7494 -0.6388 0>, rot=<0 0 0.034 -0.9994>}
 spine_02 {parent=spine_01, loc=<7.2518 -0.035 0>, rot=<0 0 0.0238 -0.9997>}
 spine_03 {parent=spine_02, loc=<6.0104 0 0>, rot=<0 0 0.0149 -0.9999>}
 spine_04 {parent=spine_03, loc=<8.1024 0 0>, rot=<0 0 -0.0132 -0.9999>}
 spine_05 {parent=spine_04, loc=<10.0146 0 0>, rot=<0 0 -0.0015 -1>}
]
variables [
 CrouchingFloat {type=double, default=0}
 NotCrouchingandPickaxe {type=double, default=0}
 isInFrontEndFloat {type=double, default=0}
 ismeleeAttackingFloat {type=double, default=0}
]
