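# Physics asset for the GoatFish fish pickaxe cosmetic, named as the front-end variant, giving its dynamic bones simulation bodies.
# Capsule bodies cover a spine chain from root through four segments, two pelvic fins and the tail base.
# Swing-limited constraints with zero twist limit link each spine segment to the previous one and hang both fins, with differing swing ranges, from the third segment.
# The tail base body has no constraint.
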PhysicsAsset /BRCosmetics/Weapons/FORT_Melee/Pickaxe_GoatFish/Meshes/GoatFish_Axe_Physics_FrontEnd
bodies {
 dyn_l_Pelvic_fin_02 [{shapeType=Capsule, length=13, radius=4.5, shapeName=dyn_l_Pelvic_fin_02_capsule, center=<10 0 0>, rotation=<rot -38 65.1181 -82.6436>}]
 dyn_r_Pelvic_fin_02 [{shapeType=Capsule, length=13, radius=4.6, shapeName=dyn_r_Pelvic_fin_02_capsule, center=<10 3 0>, rotation=<rot 0 105.9487 -89.043>}]
 dyn_spine_01 [{shapeType=Capsule, length=5, radius=4, shapeName=dyn_spine_01_capsule, center=<10 0 0>, rotation=<rot 90 0 0>}]
 dyn_spine_02 [{shapeType=Capsule, length=5, radius=10, shapeName=dyn_spine_02_capsule, center=<10 0 0>, rotation=<rot 90 0 0>}]
 dyn_spine_03 [{shapeType=Capsule, length=5, radius=10, shapeName=dyn_spine_03_capsule, center=<10 0 0>, rotation=<rot 90 0 0>}]
 dyn_spine_04 [{shapeType=Capsule, length=5, radius=12, shapeName=dyn_spine_04_capsule, center=<10 0 0>, rotation=<rot 90 0 0>}]
 dyn_spine_root [{shapeType=Capsule, length=5, radius=4, shapeName=dyn_spine_root_capsule, rotation=<rot 90 0 0>}]
 dyn_tail_base [{shapeType=Capsule, length=5, radius=5}]
}
constraints [
 {bone1Name=dyn_l_Pelvic_fin_02, bone2Name=dyn_spine_03, swing1LimitDegrees=50, swing2LimitDegrees=50, twistLimitDegrees=0}
 {bone1Name=dyn_r_Pelvic_fin_02, bone2Name=dyn_spine_03, swing1LimitDegrees=35, swing2LimitDegrees=35, twistLimitDegrees=0}
 {bone1Name=dyn_spine_01, bone2Name=dyn_spine_root, swing1LimitDegrees=40, swing2LimitDegrees=60, twistLimitDegrees=0}
 {bone1Name=dyn_spine_02, bone2Name=dyn_spine_01, swing1LimitDegrees=60, swing2LimitDegrees=40, twistLimitDegrees=0}
 {bone1Name=dyn_spine_03, bone2Name=dyn_spine_02, swing1LimitDegrees=60, swing2LimitDegrees=40, twistLimitDegrees=0}
 {bone1Name=dyn_spine_04, bone2Name=dyn_spine_03, swing1LimitDegrees=60, swing2LimitDegrees=40, twistLimitDegrees=0}
]
