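# Physics asset for the GoatFish pickaxe's crouching variant, pairing capsules for the fish's dyn_ bones with character pelvis, spine, leg and root capsules.
# The fish spine is a limited-swing chain, with both pelvic fins jointed to dyn_spine_03 and the head locked to dyn_spine_04; dyn_tail_base is unconstrained.
# Character spine joints are locked and leg joints free.
# A huge body_root capsule sits mostly below the origin, shaped like a floor.
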PhysicsAsset /BRCosmetics/Weapons/FORT_Melee/Pickaxe_GoatFish/Meshes/GoatFish_Axe_Physics_Crouching
bodies {
 body_root [{shapeType=Capsule, length=287.781, radius=612.219, shapeName=body_root_capsule, center=<-50.8335 0 -600.817>, rotation=<rot 90 0 0>}]
 calf_l [{shapeType=Capsule, length=32, radius=10, shapeName=thigh_l_capsule, center=<-0.3098 0.0001 -24.8782>, rotation=<rot 0 -179.9999 0.0001>}]
 calf_r [{shapeType=Capsule, length=30, radius=9.5, shapeName=thigh_r_capsule, center=<26 0 -0.0267>, rotation=<rot 90 0 0>}]
 dyn_head [{shapeType=Capsule, length=22.9466, radius=22, shapeName=dyn_head_capsule, center=<4.2888 0 9>, rotation=<rot 0 -90 -49.7389>}]
 dyn_l_Pelvic_fin_02 [{
   shapeType=Capsule
   length=15
   radius=7
   shapeName=dyn_l_Pelvic_fin_02_capsule
   center=<14.9604 -3 2.5>
   rotation=<rot 1.5667 80.1223 -81.0761>
  }]
 dyn_r_Pelvic_fin_02 [{shapeType=Capsule, length=17.6181, radius=4.6, shapeName=dyn_r_Pelvic_fin_02_capsule, center=<14 3 0>, rotation=<rot 0 105.9487 -89.043>}]
 dyn_spine_01 [{shapeType=Capsule, length=5, radius=5, shapeName=dyn_spine_01_capsule, center=<7 0 0>, rotation=<rot 90 0 0>}]
 dyn_spine_02 [{shapeType=Capsule, length=5, radius=12, shapeName=dyn_spine_02_capsule, center=<10 0 0>, rotation=<rot 90 0 0>}]
 dyn_spine_03 [{shapeType=Capsule, length=5, radius=17, shapeName=dyn_spine_03_capsule, center=<10 0 0>, rotation=<rot 90 0 0>}]
 dyn_spine_04 [{shapeType=Capsule, length=5, radius=15, shapeName=dyn_spine_04_capsule, center=<8.4341 0 0>, rotation=<rot 90 0 0>}]
 dyn_spine_root [{shapeType=Capsule, length=5, radius=4, shapeName=dyn_spine_root_capsule, rotation=<rot 90 0 0>}]
 dyn_tail_base [{shapeType=Capsule, length=5, radius=5}]
 pelvis [{shapeType=Capsule, length=0.505, radius=0.505, shapeName=pelvis_capsule, rotation=<rot 0 -90 0>}]
 spine_01 [{shapeType=Capsule, length=10, radius=22, shapeName=spine_01_capsule, center=<4 4.6635 0>, rotation=<rot 0 -90 0>}]
 spine_02 [{shapeType=Capsule, length=10, radius=22, shapeName=spine_02_capsule, center=<8.762 4.6635 0>, rotation=<rot 0 -90 0>}]
 spine_03 [{shapeType=Capsule, length=10, radius=22, shapeName=spine_03_capsule, center=<15.8333 4.6635 0>, rotation=<rot 0 -90 0>}]
 spine_04 [{shapeType=Capsule, length=10, radius=22, shapeName=spine_04_capsule, center=<15.8077 4.6635 0>, rotation=<rot 0 -90 0>}]
 thigh_l [{shapeType=Capsule, length=34.6616, radius=9.5479, shapeName=thigh_l_capsule, center=<-22.1907 -0.6066 0.3009>, rotation=<rot 90 180 0>}]
 thigh_r [{shapeType=Capsule, length=34.6616, radius=9.5479, shapeName=thigh_r_capsule, center=<22.1907 0.6066 -0.3009>, rotation=<rot 90 0 0>}]
}
constraints [
 {
  bone1Name=calf_l
  bone2Name=thigh_l
  swing1LimitDegrees=45
  swing1Motion=Free
  swing2LimitDegrees=45
  swing2Motion=Free
  twistLimitDegrees=45
  twistMotion=Free
 }
 {
  bone1Name=calf_r
  bone2Name=thigh_r
  swing1LimitDegrees=45
  swing1Motion=Free
  swing2LimitDegrees=45
  swing2Motion=Free
  twistLimitDegrees=45
  twistMotion=Free
 }
 {
  bone1Name=thigh_l
  bone2Name=pelvis
  swing1LimitDegrees=45
  swing1Motion=Free
  swing2LimitDegrees=45
  swing2Motion=Free
  twistLimitDegrees=45
  twistMotion=Free
 }
 {
  bone1Name=thigh_r
  bone2Name=pelvis
  swing1LimitDegrees=45
  swing1Motion=Free
  swing2LimitDegrees=45
  swing2Motion=Free
  twistLimitDegrees=45
  twistMotion=Free
 }
 {
  bone1Name=spine_01
  bone2Name=pelvis
  swing1LimitDegrees=45
  swing1Motion=Locked
  swing2LimitDegrees=45
  swing2Motion=Locked
  twistLimitDegrees=45
  twistMotion=Locked
 }
 {
  bone1Name=spine_02
  bone2Name=spine_01
  swing1LimitDegrees=45
  swing1Motion=Locked
  swing2LimitDegrees=45
  swing2Motion=Locked
  twistLimitDegrees=45
  twistMotion=Locked
 }
 {
  bone1Name=spine_03
  bone2Name=spine_02
  swing1LimitDegrees=45
  swing1Motion=Locked
  swing2LimitDegrees=45
  swing2Motion=Locked
  twistLimitDegrees=45
  twistMotion=Locked
 }
 {
  bone1Name=spine_04
  bone2Name=spine_03
  swing1LimitDegrees=45
  swing1Motion=Locked
  swing2LimitDegrees=45
  swing2Motion=Locked
  twistLimitDegrees=45
  twistMotion=Locked
 }
 {bone1Name=dyn_r_Pelvic_fin_02, bone2Name=dyn_spine_03, swing1LimitDegrees=35, swing2LimitDegrees=35, twistLimitDegrees=0}
 {bone1Name=dyn_spine_01, bone2Name=dyn_spine_root, swing1LimitDegrees=10, swing2LimitDegrees=60, twistLimitDegrees=0}
 {bone1Name=dyn_spine_02, bone2Name=dyn_spine_01, swing1LimitDegrees=10, swing2LimitDegrees=30, twistLimitDegrees=0}
 {bone1Name=dyn_spine_03, bone2Name=dyn_spine_02, swing1LimitDegrees=10, swing2LimitDegrees=30, twistLimitDegrees=0}
 {bone1Name=dyn_spine_04, bone2Name=dyn_spine_03, swing1LimitDegrees=10, swing2LimitDegrees=30, twistLimitDegrees=0}
 {
  bone1Name=dyn_head
  bone2Name=dyn_spine_04
  swing1LimitDegrees=45
  swing1Motion=Locked
  swing2LimitDegrees=45
  swing2Motion=Locked
  twistLimitDegrees=45
  twistMotion=Locked
 }
 {
  bone1Name=pelvis
  bone2Name=body_root
  swing1LimitDegrees=45
  swing1Motion=Free
  swing2LimitDegrees=45
  swing2Motion=Free
  twistLimitDegrees=45
  twistMotion=Free
 }
 {
  bone1Name=dyn_l_Pelvic_fin_02
  bone2Name=dyn_spine_03
  swing1LimitDegrees=15
  swing2LimitDegrees=15
  twistLimitDegrees=45
  twistMotion=Locked
 }
]
